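; BP_FaceboardCamera is a camera actor Blueprint for a player faceboard camera rig, configured entirely in its construction script.
; The script writes the camera component's relative Y location to the "Distance to faceboard" scalar in the MPC_FaceboardCamera material parameter collection, sharing it with materials.
; It then resets the component's relative transform from the Camera Offset variable, a fixed rotation and reduced scale, so the written distance predates that reset.
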(fn ConstructionScript
  (bind (_relativelocation_x _relativelocation_y) (Math|Vector|BreakVector (Variables|/Script/Engine.SceneComponent|GetRelativeLocation_X (Variables|CameraActor|GetCameraComponent))))
  (Rendering|Material|SetScalarParameterValue "/Game/Characters/Player/Common/SharedRigs/PostProcess/MPC_FaceboardCamera.MPC_FaceboardCamera" "Distance to faceboard" _relativelocation_y :!class /Script/Engine.KismetMaterialLibrary)
  (Transformation|SetRelativeTransform (Variables|CameraActor|GetCameraComponent) (Math|Transform|MakeTransform (Variables|Default|GetCamera%20Offset) "0.000000,-90.000000,0.000000" "0.300000,0.300000,0.300000") false true))
